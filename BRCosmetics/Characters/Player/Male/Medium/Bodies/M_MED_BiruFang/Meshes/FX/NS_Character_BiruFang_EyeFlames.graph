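# Niagara system that draws mesh-based eye flames on the BiruFang character, tinted by a time-of-day flame color.
# Burst-spawned particles persist past their lifetime and follow sampled skeletal mesh bones with a slight camera offset.
# Dynamic material parameters receive a random value, owner velocity along a transformed Z axis and a smoothed particle position, so the material can react to motion.
# A visibility tag follows the head-hidden flag.
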NiagaraSystem /BRCosmetics/Characters/Player/Male/Medium/Bodies/M_MED_BiruFang/Meshes/FX/NS_Character_BiruFang_EyeFlames
dependencies {
 usedDataInterfaces [
  &NiagaraDataInterfaceCurve
  &NiagaraDataInterfaceArrayFloat
  &NiagaraDataInterfaceVectorCurve
  &NiagaraDataInterfaceColorCurve
 ]
 usedDynamicInputs [
  &SetBoolByFloatComparison
  &SelectLinearColorFromArray
  &MakeIntFromFloat
  &NormalizeFloat
  &VectorLength
  &Add_Vector
  &Multiply_Vector
  &Multiply_VectorByFloat
  &NormalizeVector
 ]
 usedModules [
  &self:SystemSpawnScript.SystemScriptSource.SystemScriptGraph.NiagaraNodeAssignment_4.SetVariables_#1
  &self:SystemSpawnScript.SystemScriptSource.SystemScriptGraph.NiagaraNodeAssignment_7.SetVariables_#2
  &self:SystemSpawnScript.SystemScriptSource.SystemScriptGraph.NiagaraNodeAssignment_3.SetVariables_#3
  &SystemState
  &self:SystemSpawnScript.SystemScriptSource.SystemScriptGraph.NiagaraNodeAssignment_2.SetVariables_#4
  &self:SystemSpawnScript.SystemScriptSource.SystemScriptGraph.NiagaraNodeAssignment_0.SetVariables_#5
  &self:SystemSpawnScript.SystemScriptSource.SystemScriptGraph.NiagaraNodeAssignment_1.SetVariables_#6
  &FN_Integrator
  &self:SystemSpawnScript.SystemScriptSource.SystemScriptGraph.NiagaraNodeAssignment_6.SetVariables_#7
  &self:SystemSpawnScript.SystemScriptSource.SystemScriptGraph.NiagaraNodeAssignment_5.SetVariables_#8
 ]
}
emitters {Eye_Flames {
  emitterUpdateScript [
   &EmitterState(Life Cycle Mode=::NewEnumerator1, Loop Behavior=::NewEnumerator1, Loop Duration Mode=::NewEnumerator1, UseLoopDelay=false)
   &SpawnBurst_Instantaneous(Use Spawn Probability=false, UseLoopCountLimit=false)
  ]
  particleSpawnScript [
   &InitializeParticle(
    Color=$System.TOD Flame Color
    Lifetime=&RandomRangeFloat(Minimum=0.45, Maximum=0.6)
    Mesh Scale Mode=::NewEnumerator4
    UsePositionOffset=false
    Write Material Random=true
   )
   /SetVariables() {enabled=false, isSetParameters=true}
   &InitialMeshOrientation(
    Mesh Orientation Mode=::NewEnumerator7
    Rotation=&MakeVector(X=0, Y=0, Z=&RandomRangeFloat(Minimum=0, Maximum=1))
    Use Rotation Vector=true
   ) {enabled=false}
   &CameraOffset(Camera Offset Amount=0.5)
  ]
  particleUpdateScript [
   &ParticleState(Kill Particles When Lifetime Has Elapsed=false)
   &SkeletalMeshLocation(Bone Sampling Mode=::NewEnumerator4, MaskSpawnGroup=false, Skeletal Mesh=$User.SkeletalMesh) {module=SkeletalMeshLocation001}
   &InterpolateOverTime(Rate Of Change=5, Target Vector Value:NiagaraPosition=$Particles.Position, Variable Type=::NewEnumerator2)
   &DynamicMaterialParameters(
    Index 0 Param 1=&RandomRangeFloat(Minimum=0, Maximum=1)
    Index 0 Param 2=&Lerp_Float(
     A=1
     B=-1
     Alpha=&RemapRange(
      Input Value To Remap=&DotProduct(Vector A=$Engine.Owner.Velocity, Vector B=&TransformVector(Vector=<0 0 1>, Destination Space=::Simulation))
      Input Min=0
      Input Max=300
      Output Min=0
      Output Max=1
     )
    )
    Index 0 Param 3=&MakeFloatFromVector(Vector=<0 0 0>, Channel=::NewEnumerator0)
    Index 1 Param 1=&MakeFloatFromVector(Vector=$StackContext.InterpolateOverTime.InterpolateOverTime_V3_.Moving Average, Channel=::NewEnumerator0)
    Index 1 Param 2=&MakeFloatFromVector(Vector=$StackContext.InterpolateOverTime.InterpolateOverTime_V3_.Moving Average, Channel=::NewEnumerator1)
    Index 1 Param 3=&MakeFloatFromVector(Vector=$StackContext.InterpolateOverTime.InterpolateOverTime_V3_.Moving Average, Channel=::NewEnumerator2)
    Module.Param0WriteEnabled=false
    Module.Param0WriteEnabled1=true
    Module.Param1WriteEnabled=false
    Module.Param1WriteEnabled1=true
    Module.Param2WriteEnabled=false
    Module.Param2WriteEnabled1=true
    Module.Param3WriteEnabled=false
    Module.Param3WriteEnabled1=true
    Write Parameter Index 1=true
   ) {module=DynamicMaterialParameters001}
   &SkeletalMeshLocation(MaskSpawnGroup=false, Mesh Sampling Type=::NewEnumerator5, Skeletal Mesh=$User.SkeletalMesh, Socket Sampling Mode=::NewEnumerator5) {enabled=false}
   &DynamicMaterialParameters(
    Index 0 Param 1=&FloatFromCurve(FloatCurve=DataInterface() {propertyValues {Curve {keys [
         {arriveTangent=18.994, interpMode=RCIM_Cubic, leaveTangent=18.994, tangentMode=RCTM_User, time=0, value=-1}
         {arriveTangent=0.0227, interpMode=RCIM_Cubic, leaveTangent=-0.0245, tangentMode=RCTM_Break, time=0.1, value=0}
         {interpMode=RCIM_Cubic, leaveTangent=9.2665, tangentMode=RCTM_Break, time=0.7, value=0}
         {arriveTangent=-0.0082, interpMode=RCIM_Cubic, leaveTangent=-0.0082, tangentMode=RCTM_User, time=1, value=1}
        ]}}})
    Index 0 Param 2=0.5
    Index 0 Param 3=&FloatFromCurve(
     FloatCurve=DataInterface() {propertyValues {Curve {keys [
         {arriveTangent=37.8788, interpMode=RCIM_Cubic, leaveTangent=37.8788, tangentMode=RCTM_User, time=0, value=8}
         {interpMode=RCIM_Cubic, tangentMode=RCTM_User, time=0.1, value=9.75}
         {tangentMode=RCTM_User, time=1, value=3.75}
        ]}}}
     Scale Curve=1.6
    )
    Index 0 Param 4=&FloatFromCurve(
     FloatCurve=DataInterface() {propertyValues {Curve {keys [
         {arriveTangent=0.1537, interpMode=RCIM_Cubic, leaveTangent=0.1537, tangentMode=RCTM_User, time=0.3, value=0}
         {arriveTangent=4.1515, interpMode=RCIM_Cubic, leaveTangent=4.1515, tangentMode=RCTM_User, time=1, value=1.525}
        ]}}}
     Scale Curve=&Multiply_Float(
      A=&SelectFloatFromArray(
       Array Sampling Mode=::NewEnumerator0
       Direct Array Index=?
       Float Selection Array=DataInterface() {propertyValues {FloatData [0.5,0.4,0.15]}}
       Random Seed=0
      )
      B=&SelectFloatFromArray(
       Array Sampling Mode=::NewEnumerator0
       Direct Array Index=?
       Float Selection Array=DataInterface() {propertyValues {FloatData [-1,1]}}
       Random Seed=0
      )
     )
    )
    Index 1 Param 1=&FloatFromCurve(
     FloatCurve=DataInterface() {propertyValues {Curve {keys [{interpMode=RCIM_Cubic, tangentMode=RCTM_User, time=0.3, value=0}, {interpMode=RCIM_Cubic, tangentMode=RCTM_User, time=1, value=1}]}}}
     Scale Curve=0.01
    )
    Module.Param0WriteEnabled=true
    Module.Param0WriteEnabled1=true
    Module.Param1WriteEnabled=true
    Module.Param1WriteEnabled1=false
    Module.Param2WriteEnabled=true
    Module.Param2WriteEnabled1=false
    Module.Param3WriteEnabled=true
    Module.Param3WriteEnabled1=false
    Write Parameter Index 1=true
   ) {enabled=false}
   &ScaleColor(
    Scale Alpha=&FloatFromCurve(FloatCurve=DataInterface() {propertyValues {Curve {keys [
         {interpMode=RCIM_Cubic, tangentMode=RCTM_User, time=0, value=0}
         {interpMode=RCIM_Cubic, tangentMode=RCTM_User, time=0.1, value=1}
         <at 0.9 1>
         {tangentMode=RCTM_User, time=1, value=0}
        ]}}})
    ScaleA=true
    ScaleRGB=true
   ) {enabled=false}
   &ScaleMeshSize(Scale Factor=&VectorFromCurve(VectorCurve=DataInterface() {propertyValues {
       XCurve {keys [
         {interpMode=RCIM_Cubic, tangentMode=RCTM_User, time=0, value=1}
         {arriveTangent=0.712, interpMode=RCIM_Cubic, leaveTangent=0.712, tangentMode=RCTM_User, time=1, value=1.25}
        ]}
       YCurve {keys [
         {arriveTangent=-0.831, interpMode=RCIM_Cubic, leaveTangent=-0.831, tangentMode=RCTM_User, time=0, value=1.25}
         {arriveTangent=0.0193, interpMode=RCIM_Cubic, leaveTangent=0.0193, tangentMode=RCTM_User, time=1, value=0.75}
        ]}
       ZCurve {keys [<at 0 1>]}
      }})) {enabled=false}
   &Color(Color=&ColorFromCurve(ColorCurve=DataInterface() {propertyValues {
       AlphaCurve {keys [<at 0.003 0>,<at 0.0498 1 RCIM_Cubic>,<at 0.8696 1>,<at 0.9617 0 RCIM_Cubic>]}
       BlueCurve {keys [<at 0.4333 0.175 RCIM_Cubic>,<at 0.7821 0.5 RCIM_Cubic>]}
       GreenCurve {keys [<at 0.4333 0.175 RCIM_Cubic>,<at 0.7821 0.5 RCIM_Cubic>]}
       RedCurve {keys [<at 0.4333 0.175 RCIM_Cubic>,<at 0.7821 0.5 RCIM_Cubic>]}
      }})) {enabled=false}
   /SetVariables(Particles.Color=&MultiplyLinearColor(
     Linear Color=$Particles.Color
     Scale Factor=&MakeFloat4FromVectorAndFloat(
      "Vector (RGB)"=&VectorFromFloat(Value=&FNTimeOfDayScalar(Night Scale=0.5, Morning Scale=1, Day Scale=1.3, Evening Scale=1))
      "Float (Alpha)"=1
     )
    )) {enabled=false, isSetParameters=true}
   /SetVariables(Particles.VisibilityTag=&MakeCustomIntFromBool(Bool=$User.bisHeadHidden, True INT=1, False INT=0)) {isSetParameters=true}
  ]
  rendererClasses [&NiagaraMeshRendererProperties]
  renderers [{rendererClass=&NiagaraMeshRendererProperties, rendererIndex=0}]
 }}
summary {
 systemName=NS_Character_BiruFang_EyeFlames
 emitters [{emitterName=Eye_Flames, rendererClasses [&NiagaraMeshRendererProperties]}]
 userVariables [
  User.bIsFrontEnd {type=NiagaraBool}
  User.bIsFrontEndPreview {type=NiagaraBool}
  User.bisHeadHidden {type=NiagaraBool}
  User.ColorB {type=LinearColor, description="", defaultValue {struct=LinearColor, value=<rgba 1 1 1 1>}}
  User.EmissiveMult {type=NiagaraFloat, defaultValue=1, description=""}
  User.FlameColor {type=LinearColor, description="", defaultValue {struct=LinearColor, value=<rgba 1 0 0.8647 1.5>}}
  User.IsSkydiving {type=NiagaraBool}
  User.OpacityMult {type=NiagaraFloat, defaultValue=1, description=""}
  User.ParaGlideLeanAlpha {type=NiagaraFloat, defaultValue=0, description=""}
  User.RibbonIDOffset {type=NiagaraInt32}
  User.SkeletalMesh {
   type=NiagaraDataInterfaceSkeletalMesh
   defaultValue {
    struct=NiagaraExt_VariableValue_DataInterface
    value {dataInterface=&self:NiagaraDataInterfaceSkeletalMesh_0}
   }
  }
  User.SkeletalMeshNebula {
   type=NiagaraDataInterfaceSkeletalMesh
   defaultValue {
    struct=NiagaraExt_VariableValue_DataInterface
    value {dataInterface=&self:NiagaraDataInterfaceSkeletalMesh_1}
   }
  }
  User.VariantColorArrayA {
   type=NiagaraDataInterfaceArrayColor
   description=""
   defaultValue {
    struct=NiagaraExt_VariableValue_DataInterface
    value {
     dataInterface=&self:NiagaraDataInterfaceArrayColor_1
     dataInterfaceClass=&NiagaraDataInterfaceArrayColor
    }
   }
  }
  User.VariantID {type=NiagaraFloat, defaultValue=0, description=""}
 ]
}
